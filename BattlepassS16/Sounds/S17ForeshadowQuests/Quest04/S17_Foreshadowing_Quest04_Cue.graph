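# Sound cue for the Season 17 foreshadowing Quest 04, playing a stake-planting sound.
# A Random node picks one of four equally weighted Stake_Plant wave variations and feeds it through a Modulator for playback variation.
# It uses the shared foreshadowing quest attenuation, routes to the Fort_Quests sound class, and has a slightly reduced volume.
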SoundCue /BattlepassS16/Sounds/S17ForeshadowQuests/Quest04/S17_Foreshadowing_Quest04_Cue
root=Modulator_0
nodes [
 {id=Modulator_0, children [Random_0]}
 {id=Random_0, children [WavePlayer_0,WavePlayer_1,WavePlayer_2,WavePlayer_3], properties {weights [1,1,1,1]}}
 {id=WavePlayer_0, properties {soundWaveAssetPtr=&/S17_Foreshadowing_Quest04_Stake_Plant_01}}
 {id=WavePlayer_1, properties {soundWaveAssetPtr=&/S17_Foreshadowing_Quest04_Stake_Plant_02}}
 {id=WavePlayer_2, properties {soundWaveAssetPtr=&/S17_Foreshadowing_Quest04_Stake_Plant_03}}
 {id=WavePlayer_3, properties {soundWaveAssetPtr=&/S17_Foreshadowing_Quest04_Stake_Plant_04}}
]
properties {
 attenuationSettings=/BattlepassS16/Sounds/S17ForeshadowQuests/S17_Foreshadowing_Quest_Att
 maxDistance=2097152
 soundClassObject=&Fort_Quests
 volumeMultiplier=0.9
}
